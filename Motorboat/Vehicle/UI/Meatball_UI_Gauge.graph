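# Dashboard gauge widget for the Meatball motorboat, showing fuel, boost, speed and vehicle warnings.
# Construct caches dynamic materials for the meters and speed needle, and parent-class events write their shared fill parameter.
# Low-fuel, out-of-fuel and disabled states drive WBP_Vehicle_Warning, with async Vehicle.Disabled tag listeners started on each vehicle update.
# A separate event plays the animated "No Fuel!" message, and infinite fuel pins the meter full and hides the warning.
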WidgetBlueprint /Motorboat/Vehicle/UI/Meatball_UI_Gauge
animations [NoFuelMessage {
  binding_count=2
  length_seconds=2.7194
  object=&self:NoFuelMessage
  sequence {bindings [
    Overlay_NoFuelMessage {
     id=#1
     possessed_class=Overlay
     tracks [
      {
       type=MovieSceneByteTrack
       display_name=Visibility
       sections [{type=MovieSceneByteSection, channels [None {type=MovieSceneScriptingByteChannel, keys [[0,1],[1,3],[54.388,1]]}]}]
      }
      {
       type=MovieSceneFloatTrack
       display_name=Render Opacity
       sections [{type=MovieSceneFloatSection, channels [None {type=MovieSceneScriptingFloatChannel, keys [[1,0],[3,1],[40.662,1],[54.388,0]]}]}]
      }
     ]
    }
    CommonTextBlock_15 {
     id=#2
     possessed_class=CommonTextBlock
     tracks [{
       type=MovieScene2DTransformTrack
       sections [{
         type=MovieScene2DTransformSection
         channels [
          Translation.X {type=MovieSceneScriptingFloatChannel}
          Translation.Y {type=MovieSceneScriptingFloatChannel}
          Angle {type=MovieSceneScriptingFloatChannel}
          Scale.X {type=MovieSceneScriptingFloatChannel, keys [[3,1],[4.2057,2.7122],[6,2.9755],[8,1],[8.8983,1]]}
          Scale.Y {type=MovieSceneScriptingFloatChannel, keys [[3,1],[4.2057,2.7122],[6,2.9755],[8,1],[8.8983,1]]}
          Shear.X {type=MovieSceneScriptingFloatChannel}
          Shear.Y {type=MovieSceneScriptingFloatChannel}
         ]
        }]
      }]
    }
   ]}
 }]
graphs [
 EventGraph {nodes [
   n1 = Event OnWidgetVisibilityChanged!() -> (OutputDelegate, then, bNewVisibility:bool) {is=title K2Node_Event}
   > n2 = Branch(Condition=@n1.bNewVisibility)
   StopAnimation(execute=@n2.else, InAnimation=Get NoFuelMessage() -> (NoFuelMessage:Widget Animation Object Reference))
   n3 = Event SetVehicleHasInfiniteFuel!() -> (OutputDelegate, then, bInfiniteFuel:bool) {is=title K2Node_Event}
   > Branch(Condition=@n3.bInfiniteFuel)
   > SetScalarParameterValue(
    self=Get FuelMeterMaterial() -> (FuelMeterMaterial:Material Instance Dynamic Object Reference)
    ParameterName=Get GaugeFillParamName() -> (GaugeFillParamName:Name=None)
    Value=1
   )
   > SetVisibility(self=Get WBP_Vehicle_Warning() -> (WBP_Vehicle_Warning:WBP Vehicle Warning Object Reference), InVisibility=Collapsed)
   Event ShowNoFuelMessage() {is=title K2Node_Event}
   > PlayAnimationForward(InAnimation=Get NoFuelMessage())
   Event OnOutOfFuel() {is=title K2Node_Event}
   n4 = Get WBP_Vehicle_Warning()
   > ShowWarning!(execute, self:WBP Vehicle Warning Object Reference=@n4, Type:ENUM_WarningType Enum=NewEnumerator1) -> (then)
   n5 = Event OnUpdateBoostChargePercent!() -> (OutputDelegate, then, BoostPercent:float) {is=title K2Node_Event}
   > SetScalarParameterValue(
    self=Get BoostMeterMaterial() -> (BoostMeterMaterial:Material Instance Dynamic Object Reference)
    ParameterName=Get GaugeFillParamName() -> (GaugeFillParamName=None)
    Value=@n5.BoostPercent
   )
   n6 = Event UpdateSpeedometer!() -> (OutputDelegate, then, NormalizedSpeed:float, RawSpeed:float) {is=title K2Node_Event}
   n7 = Get GaugeFillParamName() -> (GaugeFillParamName=None)
   > SetScalarParameterValue(
    self=Get SpeedMeterMaterial() -> (SpeedMeterMaterial:Material Instance Dynamic Object Reference)
    ParameterName=@n7
    Value=@n6.NormalizedSpeed
   )
   > SetScalarParameterValue(
    self=Get SpeedNeedleMeterMaterial() -> (SpeedNeedleMeterMaterial:Material Instance Dynamic Object Reference)
    ParameterName=@n7
    Value=@n6.NormalizedSpeed
   )
   Event OnLowFuel() {is=title K2Node_Event}
   > ShowWarning(self=@n4, Type=NewEnumerator0)
   Event Construct()
   > n8 = GetDynamicMaterial(self=Get FuelMeter() -> (FuelMeter:Image Object Reference))
   > Set FuelMeterMaterial(FuelMeterMaterial:Material Instance Dynamic Object Reference=@n8) -> (Output_Get:Material Instance Dynamic Object Reference)
   > n9 = GetDynamicMaterial(self=Get BoostMeter() -> (BoostMeter:Image Object Reference))
   > Set BoostMeterMaterial(BoostMeterMaterial:Material Instance Dynamic Object Reference=@n9) -> (Output_Get:Material Instance Dynamic Object Reference)
   > n10 = GetDynamicMaterial(self=Get SpeedMeter() -> (SpeedMeter:Image Object Reference))
   > Set SpeedMeterMaterial(SpeedMeterMaterial:Material Instance Dynamic Object Reference=@n10) -> (Output_Get:Material Instance Dynamic Object Reference)
   > n11 = GetDynamicMaterial(self=Get SpeedNeedleMeter() -> (SpeedNeedleMeter:Image Object Reference))
   > Set SpeedNeedleMeterMaterial(SpeedNeedleMeterMaterial:Material Instance Dynamic Object Reference=@n11) -> (Output_Get:Material Instance Dynamic Object Reference)
   n12 = Event OnFuelValueChanged!() -> (OutputDelegate, then, FuelPercentage:float, bIsLowOnFuel:bool) {is=title K2Node_Event}
   > SetScalarParameterValue(self=Get FuelMeterMaterial(), ParameterName=Get GaugeFillParamName() -> (GaugeFillParamName=None), Value=@n12.FuelPercentage)
   n13 = Get isDisabled() -> (isDisabled:bool)
   n14 = Get WBP_Vehicle_Warning()
   > SetVisibility(self=@n14, InVisibility=Select(Option 0=Collapsed, Option 1=SelfHitTestInvisible, Index=OR Boolean(A=@n12.bIsLowOnFuel, B=@n13)))
   > Branch(Condition="float > float"(A=@n12.FuelPercentage))
   > RemoveWarning!(execute, self:WBP Vehicle Warning Object Reference=@n14, Type:ENUM_WarningType Enum=NewEnumerator1) -> (then)
   n15 = Event OnDisabledChanged!() -> (OutputDelegate, then, bDisabled:bool) {is=title K2Node_Event}
   > n16 = Set isDisabled(isDisabled:bool=@n15.bDisabled) -> (Output_Get:bool)
   > n17 = Branch(Condition=@n16)
   n18 = Get WBP_Vehicle_Warning()
   > ShowWarning(self=@n18, Type=NewEnumerator2)
   RemoveWarning(execute=@n17.else, self=@n18, Type=NewEnumerator2)
   n19 = Event OnVehicleUpdated!() -> (OutputDelegate, then, NewVehicle:Fort Athena Vehicle Object Reference) {is=title K2Node_Event}
   n20 = Get DisabledTag() -> (DisabledTag:Gameplay Tag Structure)
   > n21 = WaitGameplayTagAddToActor!(execute:Exec, TargetActor:Actor Object Reference=@n19.NewVehicle, Tag:Gameplay Tag Structure=@n20, OnlyTriggerOnce:bool=false) -> (then:Exec, AsyncTaskProxy:Ability Async Wait Gameplay Tag Added Object Reference, Added:Exec) {is=title K2Node_AsyncAction}
   > n22 = WaitGameplayTagRemoveFromActor!(execute:Exec, TargetActor:Actor Object Reference=@n19.NewVehicle, Tag:Gameplay Tag Structure=@n20, OnlyTriggerOnce:bool=false) -> (then:Exec, AsyncTaskProxy:Ability Async Wait Gameplay Tag Removed Object Reference, Removed:Exec) {is=title K2Node_AsyncAction}
   OnVehicleEnabled!(execute=@n22.Removed, self, Enabled:bool=true) -> (then)
   OnVehicleEnabled(execute=@n21.Added, Enabled=false)
  ]}
 OnVehicleEnabled {nodes [
   n1 = OnVehicleEnabled!() -> (then, Enabled:bool=~) {is=title K2Node_FunctionEntry}
   > n2 = Branch(Condition=@n1)
   n3 = Get WBP_Vehicle_Warning()
   > RemoveWarning(self=@n3, Type=NewEnumerator2)
   > SetVisibility(self=@n3, InVisibility=Collapsed)
   ShowWarning(execute=@n2.else, self=@n3, Type=NewEnumerator2)
   > SetVisibility(self=@n3, InVisibility=HitTestInvisible)
  ]}
]
variables [
 BoostMeterMaterial {type=MaterialInstanceDynamic}
 DefaultFuelWarningIconRenderOpacity {type=number, default=1}
 DisabledTag {type=GameplayTag, default {tagName=Vehicle.Disabled}}
 FuelEmptyMaterialInstance {type=MaterialInterface, default=/Game/Athena/HUD/Vehicles/Materials/M_UI_LowFuel_Empty}
 FuelMeterFillColor_Default {type=LinearColor, default=<rgba 1 1 1 1>}
 FuelMeterFillColor_LowFuel {type=LinearColor, default=<rgba 1 0.1137 0 1>}
 FuelMeterFillColorParamName {type=string, default=FillLineColor}
 FuelMeterMaterial {type=MaterialInstanceDynamic}
 GaugeFillParamName {type=string, default=GaugeFill}
 isDisabled {type=boolean}
 LowFuelAnimationTimerHandle {type=TimerHandle}
 LowFuelMaterialInstance {type=MaterialInterface, default=/Game/Athena/HUD/Vehicles/Materials/M_UI_LowFuel_Inst}
 PulseInterval {type=number, default=2}
 SpeedMeterMaterial {type=MaterialInstanceDynamic}
 SpeedNeedleMeterMaterial {type=MaterialInstanceDynamic}
 WarningIconColor_Off {type=LinearColor, default=<rgba 1 1 1 0.5>}
 WarningIconColor_On {type=LinearColor, default=<rgba 1 0.1137 0 1>}
]
widgets {
 info {
  inheritedWidgetCount=10
  parentClass=/Script/MotorboatUI.FortAthenaVehicleDashboardWidget_Meatball
  rootWidgetClass=&Overlay
  widgetCount=27
 }
 widgets [
  Overlay_160 {widgetClassPath=&Overlay, widget=Overlay_160, properties {visibility=HitTestInvisible}}
  Overlay_Gauge {
   widgetClassPath=&Overlay
   bInherited=true
   parent=Overlay_160
   widget=Overlay_Gauge
   slot {class=OverlaySlot, properties {horizontalAlignment=HAlign_Right, verticalAlignment=VAlign_Bottom}}
  }
  Overlay_FuelContainer {
   widgetClassPath=&Overlay
   bIsVariable=true
   parent=Overlay_Gauge
   widget=Overlay_FuelContainer
   slot {class=OverlaySlot, properties {horizontalAlignment=HAlign_Right, verticalAlignment=VAlign_Bottom, padding=<ltrb 0 0 25 6>}}
  }
  Overlay_FuelMeter {
   widgetClassPath=&Overlay
   parent=Overlay_FuelContainer
   widget=Overlay_FuelMeter
   slot {class=OverlaySlot, properties {verticalAlignment=VAlign_Center}}
  }
  FuelMeter {
   widgetClassPath=&Image
   bIsVariable=true
   parent=Overlay_FuelMeter
   widget=FuelMeter
   properties {
    brush {
     bIsDynamicallyLoaded=false
     drawAs=Image
     imageType=NoImage
     mirroring=NoMirror
     resourceName=None
     resourceObject=&/M_UI_Meatball_Speedometer_Inst_Fuel
     tiling=NoTile
     imageSize=<300 300>
     margin=<ltrb 0 0 0 0>
     outlineSettings {
      bUseBrushTransparency=false
      roundingType=HalfHeightRadius
      width=0
      color {colorUseRule=UseColor_Specified, specifiedColor=<rgba 0 0 0 0>}
      cornerRadii=<0 0 0 0>
     }
     tintColor {colorUseRule=UseColor_Specified, specifiedColor=<rgba 1 1 1 1>}
     uVRegion {bIsValid=false, max=<0 0>, min=<0 0>}
    }
    renderTransform {angle=0, scale=<0.95 1>, shear=<0 0>, translation=<0 0>}
   }
   slot {class=OverlaySlot, properties {horizontalAlignment=HAlign_Center, verticalAlignment=VAlign_Center}}
  }
  Image_MeterIcon {
   widgetClassPath=&Image
   bIsVariable=true
   parent=Overlay_FuelMeter
   widget=Image_MeterIcon
   properties {brush {
     bIsDynamicallyLoaded=false
     drawAs=Image
     imageType=NoImage
     mirroring=NoMirror
     resourceName=None
     resourceObject=/Game/Athena/HUD/Materials/M_UI_FuelIcon_Fill
     tiling=NoTile
     imageSize=<27 28>
     margin=<ltrb 0 0 0 0>
     outlineSettings {
      bUseBrushTransparency=false
      roundingType=HalfHeightRadius
      width=0
      color {colorUseRule=UseColor_Specified, specifiedColor=<rgba 0 0 0 0>}
      cornerRadii=<0 0 0 0>
     }
     tintColor {colorUseRule=UseColor_Specified, specifiedColor=<rgba 1 1 1 1>}
     uVRegion {bIsValid=false, max=<0 0>, min=<0 0>}
    }}
   slot {class=OverlaySlot, properties {verticalAlignment=VAlign_Bottom, padding=<ltrb 100 0 0 35>}}
  }
  Overlay_SpeedMeter {
   widgetClassPath=&Overlay
   parent=Overlay_FuelContainer
   widget=Overlay_SpeedMeter
   slot {class=OverlaySlot, properties {horizontalAlignment=HAlign_Right, verticalAlignment=VAlign_Bottom}}
  }
  SpeedMeter {
   widgetClassPath=&Image
   bIsVariable=true
   parent=Overlay_SpeedMeter
   widget=SpeedMeter
   properties {
    brush {
     bIsDynamicallyLoaded=false
     drawAs=Image
     imageType=NoImage
     mirroring=NoMirror
     resourceName=None
     resourceObject=&/M_UI_Meatball_Speedometer_Inst_Speed
     tiling=NoTile
     imageSize=<350 350>
     margin=<ltrb 0 0 0 0>
     outlineSettings {
      bUseBrushTransparency=false
      roundingType=HalfHeightRadius
      width=0
      color {colorUseRule=UseColor_Specified, specifiedColor=<rgba 0 0 0 0>}
      cornerRadii=<0 0 0 0>
     }
     tintColor {colorUseRule=UseColor_Specified, specifiedColor=<rgba 1 1 1 1>}
     uVRegion {bIsValid=false, max=<0 0>, min=<0 0>}
    }
    renderTransform {angle=0, scale=<0.95 0.95>, shear=<0 0>, translation=<0 0>}
   }
   slot {class=OverlaySlot, properties {horizontalAlignment=HAlign_Right, verticalAlignment=VAlign_Bottom}}
  }
  NumericText_Speed {
   widgetClassPath=/Script/CommonUI.CommonNumericTextBlock
   bInherited=true
   bIsVariable=true
   parent=Overlay_SpeedMeter
   widget=NumericText_Speed
   properties {
    bIsVolatile=true
    justification=Center
    performSizeInterpolation=false
    postInterpolationShrinkDuration=0
    text=Text Block
    font {
     bForceMonospaced=false
     bMaterialIsStencil=false
     fontMaterial=null
     fontObject=&Burbank
     letterSpacing=0
     monospacedWidth=1
     size=65
     skewAmount=0
     typefaceFontName=BurbankRegular-Black
     outlineSettings {
      bApplyOutlineToDropShadows=false
      bMiteredCorners=false
      bSeparateFillAlpha=false
      outlineMaterial=null
      outlineSize=0
      outlineColor=<rgba 0 0 0 1>
     }
    }
    renderTransform {angle=0, scale=<1 1>, shear=<-8 0>, translation=<0 -5>}
    shadowOffset=<0 0>
   }
   slot {class=OverlaySlot, properties {horizontalAlignment=HAlign_Center, verticalAlignment=VAlign_Center, padding=<ltrb 0 12 0 0>}}
  }
  SpeedNeedleMeter {
   widgetClassPath=&Image
   bIsVariable=true
   parent=Overlay_SpeedMeter
   widget=SpeedNeedleMeter
   properties {
    brush {
     bIsDynamicallyLoaded=false
     drawAs=Image
     imageType=NoImage
     mirroring=NoMirror
     resourceName=None
     resourceObject=&/M_UI_Meatball_Speedometer_Inst_SpeedNeedle
     tiling=NoTile
     imageSize=<350 350>
     margin=<ltrb 0 0 0 0>
     outlineSettings {
      bUseBrushTransparency=false
      roundingType=HalfHeightRadius
      width=0
      color {colorUseRule=UseColor_Specified, specifiedColor=<rgba 0 0 0 0>}
      cornerRadii=<0 0 0 0>
     }
     tintColor {colorUseRule=UseColor_Specified, specifiedColor=<rgba 1 1 1 1>}
     uVRegion {bIsValid=false, max=<0 0>, min=<0 0>}
    }
    renderTransform {angle=0, scale=<0.5 0.5>, shear=<0 0>, translation=<0 0>}
   }
   slot {class=OverlaySlot, properties {horizontalAlignment=HAlign_Center, verticalAlignment=VAlign_Center}}
  }
  CommonTextBlock_72 {
   widgetClassPath=&CommonTextBlock
   parent=Overlay_SpeedMeter
   widget=CommonTextBlock_72
   properties {
    text=KPH
    visibility=Hidden
    font {
     bForceMonospaced=false
     bMaterialIsStencil=false
     fontMaterial=null
     fontObject=&Burbank
     letterSpacing=0
     monospacedWidth=1
     size=15
     skewAmount=0
     typefaceFontName=BurbankRegular-Black
     outlineSettings {
      bApplyOutlineToDropShadows=false
      bMiteredCorners=false
      bSeparateFillAlpha=false
      outlineMaterial=null
      outlineSize=0
      outlineColor=<rgba 0 0 0 1>
     }
    }
    renderTransform {angle=0, scale=<1 1>, shear=<-8 0>, translation=<-3 47>}
    renderTransformPivot=<0.5 0>
    shadowOffset=<0 0>
   }
   slot {class=OverlaySlot, properties {horizontalAlignment=HAlign_Center, verticalAlignment=VAlign_Center, padding=<ltrb 0 12 0 0>}}
  }
  Overlay_Boost {
   widgetClassPath=&Overlay
   parent=Overlay_FuelContainer
   widget=Overlay_Boost
   slot {class=OverlaySlot, properties {horizontalAlignment=HAlign_Right, verticalAlignment=VAlign_Center}}
  }
  BoostMeter {
   widgetClassPath=&Image
   bIsVariable=true
   parent=Overlay_Boost
   widget=BoostMeter
   properties {
    brush {
     bIsDynamicallyLoaded=false
     drawAs=Image
     imageType=NoImage
     mirroring=NoMirror
     resourceName=None
     resourceObject=&/M_UI_Meatball_Speedometer_Inst_Boost
     tiling=NoTile
     imageSize=<300 300>
     margin=<ltrb 0 0 0 0>
     outlineSettings {
      bUseBrushTransparency=false
      roundingType=HalfHeightRadius
      width=0
      color {colorUseRule=UseColor_Specified, specifiedColor=<rgba 0 0 0 0>}
      cornerRadii=<0 0 0 0>
     }
     tintColor {colorUseRule=UseColor_Specified, specifiedColor=<rgba 1 1 1 1>}
     uVRegion {bIsValid=false, max=<0 0>, min=<0 0>}
    }
    renderTransform {angle=0, scale=<-0.95 1>, shear=<0 0>, translation=<0 0>}
   }
   slot {class=OverlaySlot, properties {horizontalAlignment=HAlign_Right, verticalAlignment=VAlign_Center}}
  }
  Image_BoostIcon {
   widgetClassPath=&Image
   bIsVariable=true
   parent=Overlay_Boost
   widget=Image_BoostIcon
   properties {brush {
     bIsDynamicallyLoaded=false
     drawAs=Image
     imageType=NoImage
     mirroring=NoMirror
     resourceName=None
     resourceObject=/Game/Athena/HUD/Vehicles/Materials/M_UI_Boost
     tiling=NoTile
     imageSize=<120 120>
     margin=<ltrb 0 0 0 0>
     outlineSettings {
      bUseBrushTransparency=false
      roundingType=HalfHeightRadius
      width=0
      color {colorUseRule=UseColor_Specified, specifiedColor=<rgba 0 0 0 0>}
      cornerRadii=<0 0 0 0>
     }
     tintColor {colorUseRule=UseColor_Specified, specifiedColor=<rgba 1 1 1 1>}
     uVRegion {bIsValid=false, max=<0 0>, min=<0 0>}
    }}
   slot {class=OverlaySlot, properties {horizontalAlignment=HAlign_Right, verticalAlignment=VAlign_Bottom, padding=<ltrb 0 0 45 -15>}}
  }
  WBP_Vehicle_Warning {
   widgetClassPath=/Game/Athena/HUD/Vehicles/Dashboard/WBP_Vehicle_Warning
   bIsVariable=true
   parent=Overlay_Gauge
   widget=WBP_Vehicle_Warning
   properties {renderTransform {angle=0, scale=<0.75 0.75>, shear=<0 0>, translation=<-374 -257>}}
   slot {class=OverlaySlot, properties {horizontalAlignment=HAlign_Right, verticalAlignment=VAlign_Bottom}}
  }
  Overlay_NoFuelMessage {
   widgetClassPath=&Overlay
   bIsVariable=true
   parent=Overlay_160
   widget=Overlay_NoFuelMessage
   properties {visibility=Collapsed, renderTransform {angle=0, scale=<1 1>, shear=<-12 0>, translation=<-80 -150>}}
   slot {class=OverlaySlot, properties {horizontalAlignment=HAlign_Center, verticalAlignment=VAlign_Bottom}}
  }
  Image_darken {
   widgetClassPath=&Image
   bIsVariable=true
   parent=Overlay_NoFuelMessage
   widget=Image_darken
   properties {
    renderOpacity=0.5
    brush {
     bIsDynamicallyLoaded=false
     drawAs=Image
     imageType=NoImage
     mirroring=NoMirror
     resourceName=None
     resourceObject=/Game/Athena/HUD/LTM/Wax/Materials/M_UI_TextGlow
     tiling=NoTile
     imageSize=<32 32>
     margin=<ltrb 0 0 0 0>
     outlineSettings {
      bUseBrushTransparency=false
      roundingType=HalfHeightRadius
      width=0
      color {colorUseRule=UseColor_Specified, specifiedColor=<rgba 0 0 0 0>}
      cornerRadii=<0 0 0 0>
     }
     tintColor {colorUseRule=UseColor_Specified, specifiedColor=<rgba 0 0 0 1>}
     uVRegion {bIsValid=false, max=<0 0>, min=<0 0>}
    }
   }
   slot {class=OverlaySlot, properties {horizontalAlignment=HAlign_Fill, verticalAlignment=VAlign_Fill, padding=<ltrb 0 -50 0 -50>}}
  }
  Image_messageBG {
   widgetClassPath=&Image
   bIsVariable=true
   parent=Overlay_NoFuelMessage
   widget=Image_messageBG
   properties {
    visibility=HitTestInvisible
    brush {
     bIsDynamicallyLoaded=false
     drawAs=Image
     imageType=NoImage
     mirroring=NoMirror
     resourceName=None
     resourceObject=/Game/Athena/HUD/Vehicles/Materials/M_UI_FuelWarningMessageBG_Inst
     tiling=NoTile
     imageSize=<32 32>
     margin=<ltrb 0 0 0 0>
     outlineSettings {
      bUseBrushTransparency=false
      roundingType=HalfHeightRadius
      width=0
      color {colorUseRule=UseColor_Specified, specifiedColor=<rgba 0 0 0 0>}
      cornerRadii=<0 0 0 0>
     }
     tintColor {colorUseRule=UseColor_Specified, specifiedColor=<rgba 1 1 1 1>}
     uVRegion {bIsValid=false, max=<0 0>, min=<0 0>}
    }
   }
   slot {class=OverlaySlot, properties {horizontalAlignment=HAlign_Fill, verticalAlignment=VAlign_Fill}}
  }
  CommonTextBlock_15 {
   widgetClassPath=&CommonTextBlock
   parent=Overlay_NoFuelMessage
   widget=CommonTextBlock_15
   properties {
    justification=Center
    text="No Fuel!"
    textTransformPolicy=ToUpper
    font {
     bForceMonospaced=false
     bMaterialIsStencil=false
     fontMaterial=null
     fontObject=&Burbank
     letterSpacing=0
     monospacedWidth=1
     size=24
     skewAmount=0
     typefaceFontName=BurbankRegular-Black
     outlineSettings {
      bApplyOutlineToDropShadows=true
      bMiteredCorners=false
      bSeparateFillAlpha=true
      outlineMaterial=/Game/UI/Foundation/Materials/M-FontShadow
      outlineSize=3
      outlineColor=<rgba 0 0 0 0.8>
     }
    }
    shadowOffset=<0 0>
   }
   slot {class=OverlaySlot, properties {padding=<ltrb 200 0 200 0>}}
  }
  Widget_FuelGauge {widgetClassPath=/Script/FortniteUI.FortVehicleFuelGauge, bInherited=true, widget=null}
  HorizontalBox_FuelCluster {widgetClassPath=&HorizontalBox, bInherited=true, widget=null}
  Image_Speedometer {widgetClassPath=&Image, bInherited=true, widget=null}
  Image_BoostGlow {widgetClassPath=&Image, bInherited=true, widget=null}
  Image_BoostMeter {widgetClassPath=&Image, bInherited=true, widget=null}
  Overlay_BoostSparks {widgetClassPath=&Overlay, bInherited=true, widget=null}
  Text_ItemName {widgetClassPath=&CommonTextBlock, bInherited=true, widget=null}
  TutorialHighlight {widgetClassPath=&TutorialHighlight}
 ]
}
